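# M_UI_MMObj_RiftingIn is a translucent UI material that draws an animated "Rifting In" icon, apparently for a minimap objective marker.
# Square edge masks and a rotated-UV reroute build static, center and border diamonds, with time-driven scaling and fading whose directions are set by static switches.
# A channel-packed texture supplies the colored icon from red and its shape from green, plus a sine-scaled center glow.
# Lerps, Max and Add composite these named-reroute layers.
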Material /DemonArenaGameplay/Gameplay/UI/Materials/M_UI_MMObj_RiftingIn
graph {
 properties {blendMode=BLEND_Translucent, materialDomain=MD_UI, pixelDepthOffsetMode=PDOM_Legacy, usedShadingModels=Default Lit}
 nodes [
  n1 = VectorParameter() {parameterName=BorderColor, defaultValue=<rgba 0 0.7157 1 1>}
  n2 = VectorParameter() {parameterName=CenterColor, defaultValue=<rgba 0.0098 0.4057 0.5923 1>}
  n3 = NamedRerouteUsage()
  n4 = NamedRerouteUsage()
  n5 = OneMinus(None=@n4)
  n6 = Max(A=NamedRerouteUsage(), B=NamedRerouteUsage())
  Add(
   A=LinearInterpolate(A=NamedRerouteUsage(), B=LinearInterpolate(A=@n1.RGB, B=@n2.RGB, Alpha=@n3), Alpha=Multiply(A=@n5, B=@n6))
   B=NamedRerouteUsage()
  )
  Max(A=@n4, B=@n6)
  n7 = TextureSampleParameter2D(UVs=MaterialFunctionCall(Texture Scale=Constant() {r=0.7}) {materialFunction=&ScaleUVsByCenter}.UVs) {
   parameterName=Icon
   samplerSource=SSM_Clamp_WorldGroupSettings
   texture=/DemonArenaGameplay/Gameplay/UI/Textures/T_UI_SkyPOI_RiftingIn_ChannelPack
  }
  n8 = Add(A=@n7.G, B=@n7.G)
  Multiply(A=@n8)
  n9 = Frac(None=Time())
  n10 = OneMinus(None=@n9)
  n11 = StaticSwitchParameter(True=@n10, False=@n9) {parameterName=Pulse Direction}
  n12 = MaterialFunctionCall(
   UVs=NamedRerouteUsage()
   Texture Scale=LinearInterpolate(B=ScalarParameter() {defaultValue=0.85, parameterName=ScaleEnd}, Alpha=@n11) {constA=0.4}
  ) {materialFunction=&ScaleUVsByCenter}
  NamedRerouteDeclaration(None=SmoothStep(Value=MaterialFunctionCall!(UVs=@n12.UVs, Mult, Exp) {materialFunction=/Game/Effects/Fort_Effects/MaterialFunctions/MF_SquareEdgeMask}.SquareMask) {constMin=0.5}) {name=CenterDiamond, nodeColor=<rgba 0.9688 1 0 1>}
  n13 = NamedRerouteUsage()
  n14 = LinearInterpolate(B=Constant() {r=0.9}, Alpha=@n11) {constA=0.5}
  n15 = SmoothStep(Value=MaterialFunctionCall(UVs=MaterialFunctionCall(UVs=@n13, Texture Scale=@n14) {materialFunction=&ScaleUVsByCenter}.UVs) {materialFunction=&MF_SquareEdgeMask}.SquareMask) {constMin=0.5}
  n16 = NamedRerouteUsage()
  NamedRerouteDeclaration(None=Multiply(A=Subtract(A=@n15, B=@n16), B=StaticSwitchParameter(True=@n9, False=@n10) {parameterName=Fade Direction})) {name=DiamondBorder, nodeColor=<rgba 0 0.0625 1 1>}
  n17 = Time()
  n18 = Add(A=Sine(None=Multiply(A=@n17) {constB=0.5})) {constB=3}
  n19 = TextureSampleParameter2D(UVs=MaterialFunctionCall(Texture Scale=Multiply(A=@n18) {constB=0.25}) {materialFunction=&ScaleUVsByCenter}.UVs) {
   parameterName=Icon
   samplerSource=SSM_Clamp_WorldGroupSettings
   texture=&T_UI_SkyPOI_RiftingIn_ChannelPack
  }
  NamedRerouteDeclaration(None=Multiply(A=VectorParameter() {parameterName=CenterGlowColor, defaultValue=<rgba 0.7705 1.6914 3 1>}.RGB, B=@n19.G)) {name=CenterGlow, nodeColor=<rgba 0.0078 0 1 1>}
  NamedRerouteDeclaration(None=@n8) {name=IconShape, nodeColor=<rgba 0 0.3672 1 1>}
  n20 = Constant() {r=90}
  NamedRerouteDeclaration(None=MaterialFunctionCall(
    UVs=TextureCoordinate()
    "Rotation Angle (0-1)"=MaterialFunctionCall!(Radians=@n20) {materialFunction=/Engine/Functions/Engine_MaterialFunctions01/Math/RadiansToDegrees}.Degrees
   ) {materialFunction=&CustomRotator}.Rotated Values) {name=DegreeRotation, nodeColor=<rgba 0 1 0.8594 1>}
  n21 = MaterialFunctionCall(UVs=NamedRerouteUsage(), Texture Scale=ScalarParameter() {defaultValue=0.5, parameterName=StaticDiamondScale}) {materialFunction=&ScaleUVsByCenter}
  NamedRerouteDeclaration(None=SmoothStep(Value=MaterialFunctionCall(UVs=@n21.UVs) {materialFunction=&MF_SquareEdgeMask}.SquareMask) {constMin=0.5}) {name=StaticDiamond, nodeColor=<rgba 1 0 0.75 1>}
  n22 = VectorParameter() {parameterName=IconGlowColor, defaultValue=<rgba 0 0.0212 0.0467 1>}
  n23 = SmoothStep(Value=@n7.R) {constMax=0.1}
  NamedRerouteDeclaration(None=LinearInterpolate(
    A=LinearInterpolate(A=@n22.RGB, Alpha=@n23)
    B=VectorParameter() {parameterName=IconColor, defaultValue=<rgba 0.1162 0.5778 0.9734 1>}.RGB
    Alpha=SmoothStep(Value=@n7.R) {constMin=0.5}
   )) {name=IconColor, nodeColor=<rgba 1 0.5156 0 1>}
 ]
}
